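# M_ToyFox_AbilityIcon is a translucent UI-domain material for ToyFox ability icons.
# Its graph holds a single texture sample parameter, SlotTexture, which lets material instances supply the icon image.
# The parameter defaults to the WorldIcon_X HUD texture, and the file shows no connection from it to the material outputs.
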Material /ToyFoxGameplay/UI/Icons/M_ToyFox_AbilityIcon
graph {
 properties {bAutomaticallySetUsageInEditor=false, blendMode=BLEND_Translucent, materialDomain=MD_UI, usedShadingModels=Default Lit}
 nodes [TextureSampleParameter2D() {parameterName=SlotTexture, texture=/Game/UI/Fort_HUD/WorldIcon_X}]
}
